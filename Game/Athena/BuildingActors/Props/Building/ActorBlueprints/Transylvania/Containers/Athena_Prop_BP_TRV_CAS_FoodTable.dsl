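; Food table building prop for the Transylvania set that sometimes attracts flies.
; On BeginPlay a random roll gives a minority chance to play the fly sound and spawn a Niagara fly swarm above the actor, storing the component.
; Damage and death effect events deactivate the swarm and stop audio, while Destroyed also destroys the Niagara component.
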(event EventOnDamagePlayEffects
  (params (Damage :type Float%20%28single-precision%29) (DamageTags :type Gameplay%20Tag%20Container%20Structure%20%28by%20ref%29) (Momentum :type Vector :default "0, 0, 0") (HitInfo :type Hit%20Result%20Structure%20%28by%20ref%29) (InstigatedBy :type Fort%20Pawn%20Object%20Reference) (DamageCauser :type Actor%20Object%20Reference) (EffectContext :type Gameplay%20Effect%20Context%20Handle%20Structure))
  ;[1 1 1] Audio
  (Components|Activation|Deactivate (Variables|Default|GetNS_Flies))
  (Audio|Components|Audio|Stop))

(event EventDestroyed
  (bind _ns_flies (Variables|Default|GetNS_Flies))
  ;[1 1 1] Audio
  (Components|Activation|Deactivate _ns_flies)
  (Components|DestroyComponent _ns_flies)
  (Audio|Components|Audio|Stop))

(event EventOnDeathPlayEffects
  (params (Damage :type Float%20%28single-precision%29) (DamageTags :type Gameplay%20Tag%20Container%20Structure%20%28by%20ref%29) (Momentum :type Vector :default "0, 0, 0") (HitInfo :type Hit%20Result%20Structure%20%28by%20ref%29) (InstigatedBy :type Fort%20Pawn%20Object%20Reference) (DamageCauser :type Actor%20Object%20Reference) (EffectContext :type Gameplay%20Effect%20Context%20Handle%20Structure))
  ;[1 1 1] Audio
  (Components|Activation|Deactivate (Variables|Default|GetNS_Flies))
  (Audio|Components|Audio|Stop))

(event EventBeginPlay
  (if (< (Math|Random|RandomFloatinRange 0.0 100) 30)
    (Audio|Components|Audio|Play (Variables|Default|GetSound_Flies) 0.0)
    (bind _returnvalue (Niagara|SpawnSystematLocation "/Game/Packages/Fortress_Transylvania/Haunted_Forest/Swamps/Particles/NS_Flies_01.NS_Flies_01" (+ (Transformation|GetWorldLocation (Variables|Transformation|GetRootComponent)) "0,0.000000,176.000000")))
    (Variables|Default|SetNS_Flies _returnvalue)))

(fn ConstructionScript)
