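; Kitchen set building prop for the Transylvania set that plays armoire impact sounds on damage and death, with an optional Halloween sound scare.
; On authority, BeginPlay stores the spawn location and, if HalloweenScript is set, binds the trigger's overlap.
; Player pawn overlaps get a limited number of DropChance rolls, and a success calls SpawnSound once.
; SpawnSound plays a fridge search sound at the stored location only on its non-authority branch.
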(event EventOnDamagePlayEffects
  (params (Damage :type Float%20%28single-precision%29) (DamageTags :type Gameplay%20Tag%20Container%20Structure%20%28by%20ref%29) (Momentum :type Vector :default "0, 0, 0") (HitInfo :type Hit%20Result%20Structure%20%28by%20ref%29) (InstigatedBy :type Fort%20Pawn%20Object%20Reference) (DamageCauser :type Actor%20Object%20Reference) (EffectContext :type Gameplay%20Effect%20Context%20Handle%20Structure))
  ;[1 1 1] Audio
  (Audio|PlaySoundatLocation "/Game/Sounds/Fort_Impact_Sounds/Demolition/Building_Props/Demo_Armoire_Impact_Cue.Demo_Armoire_Impact_Cue" (Transformation|GetActorLocation)))

(event EventOnDeathPlayEffects
  (params (Damage :type Float%20%28single-precision%29) (DamageTags :type Gameplay%20Tag%20Container%20Structure%20%28by%20ref%29) (Momentum :type Vector :default "0, 0, 0") (HitInfo :type Hit%20Result%20Structure%20%28by%20ref%29) (InstigatedBy :type Fort%20Pawn%20Object%20Reference) (DamageCauser :type Actor%20Object%20Reference) (EffectContext :type Gameplay%20Effect%20Context%20Handle%20Structure))
  ;[1 1 1] Audio
  (Audio|PlaySoundatLocation "/Game/Sounds/Fort_Impact_Sounds/Demolition/Building_Props/Demo_Armoire_Impact_Cue.Demo_Armoire_Impact_Cue" (Transformation|GetActorLocation)))

(event EventBeginPlay
  (Networking|SwitchHasAuthority
    (:Authority
      (Variables|Default|SetSpawnLocation (Transformation|GetActorLocation self))
      (if (Variables|Default|GetHalloweenScript)
        (Collision|BindEventtoOnComponentBeginOverlap (Variables|Default|GetHalloweenTrigger) (delegate Custom|Overlap))))))

(event Custom|Overlap
  (params (OverlappedComponent :type Primitive%20Component%20Object%20Reference) (OtherActor :type Actor%20Object%20Reference) (OtherComp :type Primitive%20Component%20Object%20Reference) (OtherBodyIndex :type Integer) (bFromSweep :type Boolean) (SweepResult :type Hit%20Result%20Structure%20%28by%20ref%29))
  (Utilities|Casting|CastToFortPlayerPawn OtherActor)
  (Utilities|FlowControl|DoN 5
    (:Exit
      (bind _dropresult (Class|MissionBlueprintFunctionLibrary|DropChance 10))
      (if _dropresult
        (Utilities|FlowControl|DoOnce false
          (:Completed
            (CallFunction|SpawnSound)))))))

(event Custom|SpawnSound
  (Networking|SwitchHasAuthority
    (:Remote
      (Utilities|IsValid self
        (:Is%20Valid
          (Audio|SpawnSoundatLocation :Sound "/Game/Sounds/Foley_Loot/Containers/Kitchen/Prop_Fridge_Search_Cue.Prop_Fridge_Search_Cue" :Location (Variables|Default|GetSpawnLocation) :VolumeMultiplier 0.85 :PitchMultiplier 1.0 :StartTime 0.0 :AttenuationSettings "/Game/Sounds/Attenuation/Fort_Doors_Att.Fort_Doors_Att" :bAutoDestroy true))))))

(fn ConstructionScript)
