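; Interior toilet prop Blueprint that adds random fly ambience and a Halloween scare sound.
; On BeginPlay it may randomly start Niagara flies with buzzing audio, which stop on damage or destruction.
; With HalloweenScript set, the authority binds overlap, so a player pawn can trigger SpookySound once on a DropChance roll.
; SpookySound plays a running-water loop only on the non-authority branch, using timer fade-out, delayed stop and EndPlay failsafe.
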(event EventDestroyed
  (bind _ns_flies (Variables|Default|GetNS_Flies))
  ;[1 1 1] AUDIO
  (Components|Activation|Deactivate _ns_flies)
  (Components|DestroyComponent _ns_flies)
  (Audio|Components|Audio|Stop))

(event EventOnDamagePlayEffects
  (params (Damage :type Float%20%28single-precision%29) (DamageTags :type Gameplay%20Tag%20Container%20Structure%20%28by%20ref%29) (Momentum :type Vector :default "0, 0, 0") (HitInfo :type Hit%20Result%20Structure%20%28by%20ref%29) (InstigatedBy :type Fort%20Pawn%20Object%20Reference) (DamageCauser :type Actor%20Object%20Reference) (EffectContext :type Gameplay%20Effect%20Context%20Handle%20Structure))
  ;[1 1 1] AUDIO
  (Components|Activation|Deactivate (Variables|Default|GetNS_Flies))
  (Audio|Components|Audio|Stop))

(event EventBeginPlay
  (Utilities|FlowControl|Sequence
    (:then_0
      (if (< (Math|Random|RandomFloatinRange 0.0 100) 20)
        (Audio|Components|Audio|Play (Variables|Default|GetSound_Flies) 0.0)
        (bind _returnvalue (Niagara|SpawnSystematLocation "/Game/Packages/Fortress_Transylvania/Haunted_Forest/Swamps/Particles/NS_Flies_01.NS_Flies_01" (+ (Transformation|GetWorldLocation (Variables|Transformation|GetRootComponent)) "0,0.000000,115.000000")))
        (Variables|Default|SetNS_Flies _returnvalue)))
    (:then_1
      (Networking|SwitchHasAuthority
        (:Authority
          (Variables|Default|SetSpawnLocation (Transformation|GetActorLocation self))
          (if (Variables|Default|GetHalloweenScript)
            (Collision|BindEventtoOnActorBeginOverlap (delegate Custom|PawnNear))))))))

(event EventEndPlay
  (params (EndPlayReason :type EEndPlayReason%20Enum :default "Destroyed"))
  ;[1 1 1] Failsafe for leaving before the timer to stop the looping sound fires
  (Audio|Components|Audio|Stop (Variables|Default|GetToiletSound)))

(event Custom|PawnNear
  (params (OverlappedActor :type Actor%20Object%20Reference) (OtherActor :type Actor%20Object%20Reference))
  (Utilities|Casting|CastToFortPlayerPawn OtherActor)
  (bind _dropresult (Class|MissionBlueprintFunctionLibrary|DropChance 10))
  (if _dropresult
    (Utilities|FlowControl|DoOnce false
      (:Completed
        (CallFunction|SpookySound)))))

(event Custom|SpookySound
  (bind _toiletsound (Variables|Default|GetToiletSound))
  (Networking|SwitchHasAuthority
    (:Remote
      (Utilities|IsValid self
        (:Is%20Valid
          (bind _returnvalue (Audio|SpawnSoundatLocation :Sound "/Game/Sounds/Foley_Loot/Containers/Toilet/water_toilet_runLoop_Cue.water_toilet_runLoop_Cue" :Location (Variables|Default|GetSpawnLocation) :VolumeMultiplier 1.0 :PitchMultiplier 1.0 :StartTime 0.0 :AttenuationSettings "/Game/Sounds/Attenuation/Fort_Doors_Att.Fort_Doors_Att" :bAutoDestroy true))
          (Variables|Default|SetToiletSound _returnvalue)
          (Utilities|Time|SetTimerbyEvent (delegate Custom|StopSound) 6)
          (Utilities|FlowControl|Delay 10)
          (Utilities|IsValid _toiletsound
            (:Is%20Valid
              (Audio|Components|Audio|Stop _toiletsound))))))))

(event Custom|StopSound
  (Audio|Components|Audio|FadeOut (stale Custom|SpookySound _returnvalue) 1.5 0.0 "Linear" :!class /Script/Engine.AudioComponent))

(fn ConstructionScript)
